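# Looping night interior room-tone ambience cue for the Suburban07 house foundation mix, with 3x3 square map attenuation and an in-building sound class.
# The first quality level mixes two pitch-modulated small-house loops, the second at half volume, under bell and reversed-bell 100-second looping volume envelopes.
# The second quality level uses one loop with a 30-second pitch and volume drift envelope; the remaining levels are empty.
# Overridden concurrency prevents new instances once the limit is reached.
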SoundCue /Figment_S08/Sounds/Fort_Ambient_Looping/Foundation_Mixes/FOUND_House_Suburban07_Night_Cue
root=QualityLevel_0
nodes [
 {
  id=Enveloper_0
  children [WavePlayer_2]
  properties {
   bLoop=true
   bLoopIndefinitely=true
   loopEnd=30
   pitchCurve {editorCurveData {keys [
      {interpMode=RCIM_Cubic, time=0}
      {arriveTangent=0.0002, interpMode=RCIM_Cubic, leaveTangent=0.0002, tangentMode=RCTM_User, time=7.9875, value=1.04}
      {arriveTangent=0, interpMode=RCIM_Cubic, leaveTangent=0, tangentMode=RCTM_User, time=20.0737, value=0.92}
      {interpMode=RCIM_Cubic, time=30}
     ]}}
   volumeCurve {editorCurveData {keys [<at 0 1 RCIM_Cubic>,<at 15.6539 0.8429 RCIM_Cubic>,<at 30 1 RCIM_Cubic>]}}
  }
 }
 {
  id=Enveloper_42
  children [Modulator_42]
  properties {
   bLoop=true
   bLoopIndefinitely=true
   loopEnd=100
   pitchCurve {editorCurveData {keys [{time=0}]}}
   volumeCurve {externalCurve=/Game/Sounds/Trans_Enveloper_Curves/Bell_100s, editorCurveData {keys [<at 0 1>]}}
  }
 }
 {
  id=Enveloper_43
  children [Modulator_43]
  properties {
   bLoop=true
   bLoopIndefinitely=true
   loopEnd=100
   pitchCurve {editorCurveData {keys [{time=0}]}}
   volumeCurve {externalCurve=/Game/Sounds/Trans_Enveloper_Curves/BellReversed_100s.BellReversed_100S, editorCurveData {keys [<at 0 1>]}}
  }
 }
 {id=Mixer_20, children [Enveloper_42,Enveloper_43], properties {inputVolume [1,0.5]}}
 {id=Modulator_0, children [Enveloper_0], properties {pitchMax=1.1, pitchMin=0.85, volumeMax=1, volumeMin=0.9}}
 {id=Modulator_42, children [WavePlayer_25], properties {pitchMax=1, pitchMin=0.8, volumeMax=1.1, volumeMin=0.9}}
 {id=Modulator_43, children [WavePlayer_1], properties {pitchMax=1.3, pitchMin=1.1, volumeMax=1.1, volumeMin=0.9}}
 {id=QualityLevel_0, children [Mixer_20,Modulator_0,null,null]}
 {id=WavePlayer_1, properties {bLooping=true, soundWaveAssetPtr=/Game/Sounds/Trans_AmbientLoops/RoomTones/SmallHouseA014_Loop_Stereo}}
 {id=WavePlayer_2, properties {bLooping=true, soundWaveAssetPtr=&SmallHouseA014_Loop_Stereo}}
 {id=WavePlayer_25, properties {bLooping=true, soundWaveAssetPtr=/Game/Sounds/Trans_AmbientLoops/RoomTones/SmallHouseB007_Stereo}}
]
properties {
 attenuationSettings=/Game/Sounds/Attenuation/Fort_Map_3x3_Square
 bOverrideConcurrency=true
 maxDistance=1780
 soundClassObject=/Figment_S08/Sounds/FortSoundClassesAndModes/Foundations_Ambience_InBuilding_PSM
 volumeMultiplier=1
 concurrencyOverrides {resolutionRule=PreventNew}
}
